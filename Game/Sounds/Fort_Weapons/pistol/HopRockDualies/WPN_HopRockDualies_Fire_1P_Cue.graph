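# Sound cue for the first-person fire sound of the HopRock Dualies pistols, capped by first-person gunfire concurrency.
# It mixes a close layer of 1P shots, classed as 3P PSM despite the cue's 1P class, with a distant-classed branch blending mid and quieter far shots.
# Each layer has its own attenuation and modulator and picks randomly among four equally weighted waves.
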SoundCue /Game/Sounds/Fort_Weapons/pistol/HopRockDualies/WPN_HopRockDualies_Fire_1P_Cue
root=Mixer_1
nodes [
 {
  id=Attenuation_10
  children [Modulator_14]
  properties {attenuationSettings=&Fort_Weapon_Projectile_Medium_Mono_Athena_Att}
 }
 {
  id=Attenuation_11
  children [Modulator_13]
  properties {attenuationSettings=&Fort_Weapon_Projectile_Distant_Mono_Athena_Att}
 }
 {id=Attenuation_9, children [Modulator_0], properties {attenuationSettings=&Fort_Weapon_Projectile_Close_Athena_Att}}
 {id=Mixer_1, children [SoundClass_7,SoundClass_6], properties {inputVolume [1,1]}}
 {id=Mixer_2, children [Attenuation_10, Attenuation_11], properties {inputVolume [1,0.4]}}
 {id=Modulator_0, children [Random_2]}
 {id=Modulator_13, children [Random_7], properties {pitchMax=1.1}}
 {id=Modulator_14, children [Random_6], properties {pitchMax=1.1, volumeMax=1, volumeMin=1}}
 {id=Random_2, children [WavePlayer_0,WavePlayer_1,WavePlayer_2,WavePlayer_3], properties {weights [1,1,1,1]}}
 {id=Random_6, children [WavePlayer_17, WavePlayer_18, WavePlayer_19, WavePlayer_20], properties {weights [1,1,1,1]}}
 {id=Random_7, children [WavePlayer_21, WavePlayer_22, WavePlayer_23, WavePlayer_24], properties {weights [1,1,1,1]}}
 {id=SoundClass_6, children [Mixer_2], properties {soundClassOverride=&Fort_WeaponRanged_Fire_Distant}}
 {
  id=SoundClass_7
  children [Attenuation_9]
  properties {soundClassOverride=&Fort_WeaponRanged_Fire_3P_PSM}
 }
 {id=WavePlayer_0, properties {soundWaveAssetPtr=&/WPN_HopRockDualies_Fire_1P_01}}
 {id=WavePlayer_1, properties {soundWaveAssetPtr=&/WPN_HopRockDualies_Fire_1P_02}}
 {id=WavePlayer_17, properties {soundWaveAssetPtr=&/WPN_HopRockDualies_Fire_Mid_01}}
 {id=WavePlayer_18, properties {soundWaveAssetPtr=&/WPN_HopRockDualies_Fire_Mid_02}}
 {id=WavePlayer_19, properties {soundWaveAssetPtr=&/WPN_HopRockDualies_Fire_Mid_03}}
 {id=WavePlayer_2, properties {soundWaveAssetPtr=&/WPN_HopRockDualies_Fire_1P_03}}
 {id=WavePlayer_20, properties {soundWaveAssetPtr=&/WPN_HopRockDualies_Fire_Mid_04}}
 {id=WavePlayer_21, properties {soundWaveAssetPtr=&/WPN_HopRockDualies_Fire_Far_01}}
 {id=WavePlayer_22, properties {soundWaveAssetPtr=&/WPN_HopRockDualies_Fire_Far_02}}
 {id=WavePlayer_23, properties {soundWaveAssetPtr=&/WPN_HopRockDualies_Fire_Far_03}}
 {id=WavePlayer_24, properties {soundWaveAssetPtr=&/WPN_HopRockDualies_Fire_Far_04}}
 {id=WavePlayer_3, properties {soundWaveAssetPtr=&/WPN_HopRockDualies_Fire_1P_04}}
]
properties {
 attenuationSettings=&Fort_Weapon_Projectile_FirstPerson_Athena_Att
 maxDistance=2097152
 soundClassObject=&Fort_WeaponRanged_Fire_1P_PSM
 volumeMultiplier=0.7
 attenuationOverrides {distanceAlgorithm=NaturalSound, falloffDistance=5000, attenuationShapeExtents {x=1000}}
 concurrencySet [&Weapon_Gunfire_FirstPerson_Conc]
}
