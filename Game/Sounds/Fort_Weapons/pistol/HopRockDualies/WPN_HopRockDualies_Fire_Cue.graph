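# Sound cue for the HopRock Dualies pistol gunshot, layering close, mid and far sounds by listener distance.
# Each layer randomly picks one of four waves, then passes through a modulator and its own projectile attenuation.
# The close layer uses 1P recordings under the 3P fire sound class, while mid and far are mixed under the Distant class.
# A root mixer combines both branches, with local and global gunfire concurrency limiting playback.
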SoundCue /Game/Sounds/Fort_Weapons/pistol/HopRockDualies/WPN_HopRockDualies_Fire_Cue
root=Mixer_1
nodes [
 {
  id=Attenuation_10
  children [Modulator_14]
  properties {attenuationSettings=&Fort_Weapon_Projectile_Medium_Mono_Athena_Att}
 }
 {
  id=Attenuation_11
  children [Modulator_13]
  properties {attenuationSettings=&Fort_Weapon_Projectile_Distant_Mono_Athena_Att}
 }
 {id=Attenuation_9, children [Modulator_3], properties {attenuationSettings=&Fort_Weapon_Projectile_Close_Athena_Att}}
 {id=Mixer_1, children [SoundClass_7,SoundClass_6], properties {inputVolume [0.7,1]}}
 {id=Mixer_2, children [Attenuation_10, Attenuation_11], properties {inputVolume [1,0.3]}}
 {id=Modulator_13, children [Random_7], properties {pitchMax=1.2, pitchMin=1}}
 {id=Modulator_14, children [Random_6], properties {pitchMax=1.1, volumeMax=1, volumeMin=1}}
 {id=Modulator_3, children [Random_4]}
 {id=Random_4, children [WavePlayer_14, WavePlayer_15, WavePlayer_16, WavePlayer_17], properties {weights [1,1,1,1]}}
 {id=Random_6, children [WavePlayer_18, WavePlayer_19, WavePlayer_20, WavePlayer_21], properties {weights [1,1,1,1]}}
 {id=Random_7, children [WavePlayer_22, WavePlayer_23, WavePlayer_24, WavePlayer_25], properties {weights [1,1,1,1]}}
 {id=SoundClass_6, children [Mixer_2], properties {soundClassOverride=&Fort_WeaponRanged_Fire_Distant}}
 {
  id=SoundClass_7
  children [Attenuation_9]
  properties {soundClassOverride=&Fort_WeaponRanged_Fire_3P_PSM}
 }
 {id=WavePlayer_14, properties {soundWaveAssetPtr=&/WPN_HopRockDualies_Fire_1P_01}}
 {id=WavePlayer_15, properties {soundWaveAssetPtr=&/WPN_HopRockDualies_Fire_1P_02}}
 {id=WavePlayer_16, properties {soundWaveAssetPtr=&/WPN_HopRockDualies_Fire_1P_03}}
 {id=WavePlayer_17, properties {soundWaveAssetPtr=&/WPN_HopRockDualies_Fire_1P_04}}
 {id=WavePlayer_18, properties {soundWaveAssetPtr=&/WPN_HopRockDualies_Fire_Mid_01}}
 {id=WavePlayer_19, properties {soundWaveAssetPtr=&/WPN_HopRockDualies_Fire_Mid_02}}
 {id=WavePlayer_20, properties {soundWaveAssetPtr=&/WPN_HopRockDualies_Fire_Mid_03}}
 {id=WavePlayer_21, properties {soundWaveAssetPtr=&/WPN_HopRockDualies_Fire_Mid_04}}
 {id=WavePlayer_22, properties {soundWaveAssetPtr=&/WPN_HopRockDualies_Fire_Far_01}}
 {id=WavePlayer_23, properties {soundWaveAssetPtr=&/WPN_HopRockDualies_Fire_Far_02}}
 {id=WavePlayer_24, properties {soundWaveAssetPtr=&/WPN_HopRockDualies_Fire_Far_03}}
 {id=WavePlayer_25, properties {soundWaveAssetPtr=&/WPN_HopRockDualies_Fire_Far_04}}
]
properties {
 maxDistance=27500
 volumeMultiplier=1
 attenuationOverrides {distanceAlgorithm=NaturalSound, falloffDistance=5000, attenuationShapeExtents {x=1000}}
 concurrencySet [&Weapon_Gunfire_Local_Conc, &Weapon_Gunfire_Global_Conc]
}
