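# Sound cue for the HopRock Dualies pistol reload, playing the whole start, insert and end foley sequence from one cue.
# A mixer layers the three waves, with fixed delays staggering the insert and end sounds and the end layer mixed slightly louder.
# Each layer passes through its own Modulator node.
# The cue uses the weapon foley class, shared reload attenuation, and a concurrency override that prevents new instances at the limit.
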SoundCue /Game/Sounds/Fort_Weapons/pistol/HopRockDualies/WPN_HopRockDualies_Reload_Start_Cue
root=Mixer_0
nodes [
 {id=Delay_0, children [WavePlayer_4], properties {delayMax=0.95, delayMin=0.95}}
 {id=Delay_1, children [WavePlayer_5], properties {delayMax=1.65, delayMin=1.65}}
 {id=Mixer_0, children [Modulator_2,Modulator_0,Modulator_1], properties {inputVolume [1,1,1.2]}}
 {id=Modulator_0, children [Delay_0]}
 {id=Modulator_1, children [Delay_1]}
 {id=Modulator_2, children [WavePlayer_3]}
 {id=WavePlayer_3, properties {soundWaveAssetPtr=&/WPN_HopRockDualies_Reload_Start_01}}
 {id=WavePlayer_4, properties {soundWaveAssetPtr=&/WPN_HopRockDualies_Reload_Insert_01}}
 {id=WavePlayer_5, properties {soundWaveAssetPtr=&/WPN_HopRockDualies_Reload_End_01}}
]
properties {
 attenuationSettings=&Fort_Weapon_Reload_Att
 bOverrideConcurrency=true
 maxDistance=2250
 soundClassObject=&Fort_WeaponFoley
 volumeMultiplier=0.752
 attenuationOverrides {falloffDistance=4100, attenuationShapeExtents {x=900}}
 concurrencyOverrides {resolutionRule=PreventNew}
}
